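# Masked material for the Tondi vehicle cosmetic wheel, adding animated, Custom Color-tinted tire and rim glow.
# MF_WheelParent_Simple_Alpha supplies base attributes; emissive is overridden, then everything passes through the vehicle damage and de-rez transition functions.
# The mask texture's green channel drives the glow, modulated by panning blurred-line and geometric textures, with Floor adding stepped highlights.
# Panning ToonyWater cloud noise blends in rim glow, and vertex color red selects between emissive layers.
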Material /VehicleCosmetics/Wheels/Tondi/Materials/M_Wheel_Tondi
graph {
 properties {
  bAutomaticallySetUsageInEditor=false
  bUseMaterialAttributes=true
  bUsedWithMeshDeformer=true
  bUsedWithSkeletalMesh=true
  blendMode=BLEND_Masked
  pixelDepthOffsetMode=PDOM_Legacy
  usedShadingModels=Default Lit
 }
 nodes [
  n1 = GetMaterialAttributes(None=MaterialFunctionCall() {materialFunction=/VehicleCosmetics/Wheels/Mage/Materials/MF_WheelParent_Simple_Alpha}.MaterialAttributes) {attributeGetTypes [#1]}
  n2 = NamedRerouteUsage()
  MaterialFunctionCall(In=MaterialFunctionCall(Material=SetMaterialAttributes(MaterialAttributes=@n1.MaterialAttributes, Emissive Color=@n2) {attributeSetTypes [#1]}) {materialFunction=&MF_BaseVehicle_Damage}.Result) {materialFunction=&MF_VehicleDeRez_Transition}
  n3 = ScalarParameter() {defaultValue=3, parameterName=TireGlowBrightness}
  n4 = TextureSampleParameter2D() {parameterName=TireEmissive, samplerType=SAMPLERTYPE_LinearColor, texture=/VehicleCosmetics/Wheels/Tondi/Textures/T_Tondi_Rim_M}
  n5 = Multiply(A=@n4.G, B=Desaturation(None=StaticComponentMaskParameter(None=@n4.G) {defaultB=true, defaultR=true, parameterName=TireRGB_Glow}))
  n6 = VectorParameter() {parameterName=Custom Color, defaultValue=<rgba 0.106 0.022 0.259 0.8>}
  n7 = Multiply(A=@n4.G, B=Desaturation(None=StaticComponentMaskParameter(None=@n4.G) {defaultG=true, parameterName=TireRGB_Color}))
  n8 = ScalarParameter() {defaultValue=3, parameterName=TreadLineIntensity}
  n9 = Add(A=Multiply(A=@n6.RGB, B=Multiply(A=@n3, B=@n5)), B=Multiply(A=@n6.RGB, B=Multiply(A=@n7, B=@n8)))
  n10 = Panner(Coordinate=TextureCoordinate()) {speedY=0.5}
  n11 = TextureSampleParameter2D(UVs=@n10) {parameterName=TireEmissive2, texture=/VehicleCosmetics/Wheels/UltrabarroHolographic/Textures/T_Blurredline}
  n12 = ScalarParameter() {defaultValue=1, parameterName=TireEmissiveGlow}
  n13 = Floor(None=Multiply(A=Multiply(A=@n11.RGB, B=@n12), B=TextureSample(UVs=@n10) {texture=/VehicleCosmetics/Wheels/_SharedTextures/FX/Skin_GeometricBM_RGB}.R))
  n14 = TextureCoordinate()
  n15 = Multiply(
   A=Multiply(A=@n11.RGB, B=@n12)
   B=TextureSample(UVs=Panner(Coordinate=@n14) {speedY=1}) {texture=&Skin_GeometricBM_RGB}.R
  )
  n16 = Multiply(A=Clamp(None=Add(A=@n13, B=@n15)) {maxDefault=3}, B=Constant() {r=20})
  n17 = Multiply(A=LinearInterpolate(A=@n9, B=Multiply(A=@n9, B=@n16), Alpha=Constant() {r=0.5}), B=@n8)
  n18 = ScalarParameter() {parameterName=GlowBrightness}
  n19 = Constant() {r=1}
  n20 = Panner(Coordinate=@n14) {speedY=-0.15}
  n21 = Panner(Coordinate=@n14) {speedY=-0.2}
  n22 = Multiply(
   A=Multiply(
    A=Multiply(A=@n4.G, B=@n19)
    B=TextureSample(UVs=@n20) {texture=/VehicleCosmetics/Boosts/Boost_ToonyWater/Texture/T_Boost_ToonyWater_Clouds_RGB_02_T}.B
   )
   B=Multiply(
    A=TextureSample(UVs=@n21) {texture=&T_Boost_ToonyWater_Clouds_RGB_02_T}.R
    B=ScalarParameter() {defaultValue=3, parameterName=RimEmiss_Intensity}
   )
  )
  n23 = Add(
   A=StaticSwitchParameter(True=Multiply(A=@n18, B=@n4.G), False=Constant())
   B=Multiply(A=Multiply(A=@n6.RGB, B=@n22), B=ScalarParameter() {defaultValue=1, parameterName=Rim Glow Effect})
  )
  NamedRerouteDeclaration(None=LinearInterpolate(A=LinearInterpolate(A=@n17, B=@n23, Alpha=@n22), B=NamedRerouteUsage(), Alpha=VertexColor().R)) {name=Emissive, nodeColor=<rgba 0 1 1 1>}
  NamedRerouteDeclaration(None=@n1.Emissive Color) {name=emissive base, nodeColor=<rgba 0.3594 0 1 1>}
 ]
}
